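; Season 12 quest ability that marks the player while riding in a vehicle tagged as MistyBop.
; It activates from gameplay events and switches on the event tag for vehicle entry or exit.
; On entry, it applies the InMistyBop effect to the owner if the vehicle's ability system component has the MistyBop tags.
; On exit, it removes effects granting the InMistyBop tags without checking the vehicle.
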(event Ability|EventActivateAbilityFromEvent
  (params (EventData :type Gameplay%20Event%20Data%20Structure%20%28by%20ref%29))
  (|CallParentFunction EventData)
  (switch Utilities|FlowControl|Switch|SwitchonGameplayTag (Utilities|Struct|BreakGameplayEventData EventData)
    (:Event.Vehicle.Entered
      (if (GameplayTags|HasAnyTags (GameplayTags|GetOwnedGameplayTags (Vehicle|GetVehicleAbilitySystemComponent (Vehicle|GetVehicleInterface (Variables|GAVars|Data|GetOwningPlayerPawn))) :!class /Script/GameplayTags.GameplayTagAssetInterface) (Variables|Default|GetT_MistyBop) true)
        (Ability|ApplyGameplayEffectToOwner (Variables|Default|GetGE_InMistyBop))))
    (:Event.Vehicle.Exit
      (Ability|RemoveGameplayEffectFromOwnerWithGrantedTags (Variables|Default|GetT_InMistyBop))))
  (Ability|EndAbility))
